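# Animation Blueprint for the front-end shoe skeleton, which poses cosmetic shoes and offsets the left and right shoe independently.
# A Chooser Player evaluates CT_Shoes_FrontEnd_Pose with this AnimBP as context, then additive Modify Bone nodes translate calf_l and calf_r.
# OnShoePositionsSet stores the two offsets, and OnGameplayTagsSet stores tags in ShoeBend, which is exposed to the chooser through that context.
# A standalone Shoes_FrontEnd_Pose sequence player is left unconnected.
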AnimBlueprint /CosmeticShoes/Assets/Anim/ABP_Shoe_FrontEnd
target_skeleton=&/Shoes_FrontEnd_Skel
graphs [
 AnimGraph {nodes [
   AnimGraphNode_ChooserPlayer_0 = Chooser Player: CT_Shoes_FrontEnd_Pose!() -> (Pose:Pose Link Structure) {
    is=title AnimGraphNode_ChooserPlayer
    evaluationFrequency=OnUpdate
    chooser {_structType=/Script/Chooser.EvaluateChooser, chooser=&/CT_Shoes_FrontEnd_Pose}
    chooserContextDefinition [
     {
      _structType=/Script/Chooser.ContextObjectTypeClass
      class=&/ABP_Shoe_FrontEnd.ABP_Shoe_FrontEnd_C
      direction=ReadWrite
     }
     {struct=/Script/Chooser.ChooserPlayerSettings, _structType=/Script/Chooser.ContextObjectTypeStruct, direction=Read}
    ]
    defaultSettings {blendOption=Linear}
   }
   n1 = "Transform (Modify) Bone - Bone: calf_l"!(
    ComponentPose=Local To Component(LocalPose=@AnimGraphNode_ChooserPlayer_0)
    Alpha=1
    Translation=Get Left Shoe Offset() -> (Left Shoe Offset:Vector)
   ) -> (Pose) {translationMode=BMM_Additive, alphaBoolBlend {blendOption=Linear}, boneToModify {boneName=calf_l}}
   n2 = Get Right Shoe Offset() -> (Right Shoe Offset:Vector)
   Output Pose(Result=Component To Local(ComponentPose="Transform (Modify) Bone - Bone: calf_r"!(ComponentPose=@n1, Alpha=1, Translation=@n2) -> (Pose) {translationMode=BMM_Additive, alphaBoolBlend {blendOption=Linear}, boneToModify {boneName=calf_r}}))
   Sequence Player 'Shoes_FrontEnd_Pose'() {bLoopAnimation=false, sequence=&/Shoes_FrontEnd_Pose}
  ]}
 AnimationBlendStackGraph_0 {
  path=AnimGraph.AnimGraphNode_ChooserPlayer_0.AnimationBlendStackGraph_0
  nodes [Output Pose!(Result:Pose Link Structure=Blend Stack Input!() -> (Pose:Pose Link Structure) {is=title AnimGraphNode_BlendStackInput}) {is=title AnimGraphNode_BlendStackResult}]
 }
 EventGraph {nodes [
   n1 = Event OnShoePositionsSet!() -> (OutputDelegate, then, LeftShoePosition:Vector, RightShoePosition:Vector) {is=title K2Node_Event}
   > n2 = Sequence()
   > Set Left Shoe Offset(Left Shoe Offset:Vector=@n1.LeftShoePosition) -> (Output_Get:Vector)
   Set Right Shoe Offset(execute=@n2.then_1, Right Shoe Offset:Vector=@n1.RightShoePosition) -> (Output_Get:Vector)
   n3 = Event OnGameplayTagsSet!() -> (OutputDelegate, then, GameplayTags:"Gameplay Tag Container Structure (by ref)") {is=title K2Node_Event}
   > Set ShoeBend(ShoeBend:Gameplay Tag Container Structure=@n3.GameplayTags) -> (Output_Get:Gameplay Tag Container Structure)
   TryGetPawnOwner()
   Event BlueprintUpdateAnimation()
  ]}
]
variables [
 Left Shoe Offset {type=Vector}
 Right Shoe Offset {type=Vector}
 ShoeBend {type=GameplayTagContainer, default {gameplayTags []}}
]
