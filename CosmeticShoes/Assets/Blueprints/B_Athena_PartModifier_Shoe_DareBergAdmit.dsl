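; Cosmetic part modifier for the DareBergAdmit shoe that animates a freeze and melt effect based on whether the player is moving.
; Velocity toggles play FreezeTimeline when moving and reverse it when stopped, with play rates derived from MeltTime and FreezeTime.
; The timeline drives FreezeIntensity on the part's mesh materials and FreezeLerp on its idle VFX.
; In the front end, tick is disabled and a delayed sequence simulates one move-and-stop cycle; otherwise tick is enabled.
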(event EventOnPostResIn
  (|CallParentFunction)
  (if (CustomCosmeticModifier|IsinFrontEnd)
    (Actor|Tick|SetActorTickEnabled false)
    (Utilities|FlowControl|Delay 3.0)
    (CallFunction|OnVelocityToggle true)
    (Utilities|FlowControl|Delay (+ (Variables|Default|GetMeltTime) 3.0))
    (CallFunction|OnVelocityToggle false)
    (Utilities|FlowControl|Delay (+ (Variables|Default|GetFreezeTime) 3.0))
    (else
      (Actor|Tick|SetActorTickEnabled true))))

(event EventOnVelocityToggle
  (params (IsMoving :type Boolean) (Delta%20Seconds :type Float%20%28double-precision%29))
  (bind _freezetimeline (Variables|B_Athena_PartModifier_Shoe_DareBergAdmit|GetFreezeTimeline))
  (if IsMoving
    (Components|Timeline|SetPlayRate _freezetimeline (/ 1.0 (Variables|Default|GetMeltTime)))
    (bind (_direction _freezeintensity) (Timeline|FreezeTimeline Play)
      (:Update
        (Rendering|Material|SetScalarParameterValueonMaterials (CustomCharacter|GetSkeletalMeshForPartOnPlayerPawn) "FreezeIntensity" _freezeintensity)
        (Effects|Components|ParticleSystem|SetFloatParameter (CallFunction|GetVFXComponent "(TagName=\"CosmeticPart.VFX.Idle\")") "FreezeLerp" _freezeintensity)))
    (else
      (Components|Timeline|SetPlayRate _freezetimeline (/ 1.0 (Variables|Default|GetFreezeTime)))
      (Timeline|FreezeTimeline Reverse))))

(fn ConstructionScript
  (|CallParentFunction))
