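; Cosmetic part modifier for the FrostChill shoes that escalates material and VFX reactivity with the player's eliminations.
; In game it watches the AthenaKills stat to drive material KillCount and staged idle-VFX flags; the frontend previews three stages on a timeline.
; Weapon changes update the idle effect's lit flag based on feet visibility.
; A landing override spawns effects only once stage three is reached and shoes are visible.
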(event EventOnWeaponChanged
  (params (NewWeapon :type Fort%20Weapon%20Object%20Reference) (OldWeapon :type Fort%20Weapon%20Object%20Reference))
  (bind _returnvalue (CallFunction|GetVFXComponent "(TagName=\"CosmeticPart.VFX.Idle\")"))
  (Utilities|IsValid NewWeapon
    (:Is%20Valid
      (bind _output_get (Variables|Default|SetbAreShoesHidden (or (Variables|HiddenPartCheck|GetbIsFootHidden) (Tags|WeaponDataHasMatchingGameplayTag NewWeapon "(TagName=\"Animation.CharacterParts.HideFX.Feet\")"))))
      (if _output_get
        (Effects|Components|ParticleSystem|SetBoolParameter _returnvalue "bisLit" false)
        (elif (>= (Variables|Default|GetLastValue) 2)
          (Effects|Components|ParticleSystem|SetBoolParameter _returnvalue "bisLit" true))))))

(event EventOnPawnLanded
  ;[0.6233 0.4542 0 0.5] OnPawnLand Override because on this shoe we only spawn the landing effect after 2 eliminations
  (if (Variables|Default|GetisStage3)
    (if (or (Variables|HiddenPartCheck|GetbIsFootHidden) (Variables|Default|GetbAreShoesHidden))
      _
      (elif (CallFunction|IsPreventLandingFXPlayingatIncorrectTimesEnabled)
        (CustomCharacter|ShouldPlayLandingFX "(GameplayTags=((TagName=\"Gameplay.Action.Player.InVehicle\"),(TagName=\"Gameplay.Action.Player.AthenaInVehicle\"),(TagName=\"Granted.Athena.EnvItem.HidingProp.GettingInProp\")))" "(GameplayTags=((TagName=\"MovementMode.Parachuting\"),(TagName=\"MovementMode.Skydiving\")))" 3e+01
          (:True
            (bind _returnvalue (CallFunction|GetVFXSpawnParameters "(TagName=\"CosmeticPart.VFX.Landing\")")
              (:Success
                (bind _returnvalue_2 (PartModifier|SpawnSystem|SpawnSystemAttachedPartModifier (Utilities|Struct|BreakCharacterPartModifierVFXSpawnParameters _returnvalue) (.AttachToComponent (Utilities|Struct|BreakCharacterPartModifierVFXSpawnParameters _returnvalue)) (.SocketName (Utilities|Struct|BreakCharacterPartModifierVFXSpawnParameters _returnvalue)) "0, 0, 0" "0, 0, 0" (.LocationType (Utilities|Struct|BreakCharacterPartModifierVFXSpawnParameters _returnvalue)))
                  (:Success
                    (CallFunction|SetShoeScalesonVFX _returnvalue_2)))))))
        (else
          _)))))

(event EventOnCosmeticStateActivated
  (params (Stat%20Name :type Name :default "None") (Stat%20Value :type Integer) (STAGE%20Num%20Being%20Activated :type Integer))
  (|CallParentFunction Stat%20Name Stat%20Value STAGE%20Num%20Being%20Activated)
  (Variables|Default|SetCurrentStage STAGE%20Num%20Being%20Activated)
  (CallFunction|UpdateMaterialParameters))

(event EventPreDestruction
  (|CallParentFunction)
  (Utilities|Time|ClearandInvalidateTimerbyHandle (Variables|Default|GetRefreshTimeHandler)))

(event EventOnPawnAndPartChanged
  (|CallParentFunction)
  (if (not (CustomCosmeticModifier|IsinFrontEnd))
    (Utilities|FlowControl|Sequence
      (:then_0
        (bind _returnvalue (Class|FortCosmeticFunctionLibrary|GetCosmeticStatValue (CustomCharacter|GetAssociatedPlayerPawn) "(TagName=\"StatManager.AthenaKills\")"))
        (bind _stage (CallFunction|StatValuetoStage _returnvalue))
        (Variables|Default|SetCurrentStage _stage)
        (CallFunction|UpdateMaterialParameters)))))

(event EventOnPostResIn
  (|CallParentFunction)
  (if (CallFunction|ShouldDisplayFrontendReactivity)
    (Timeline|Timeline_0 PlayFromStart
      (:Stage%201
        ;[1 1 1] In FE mode: Its gonna be 3 stage represents Normal > Emissive light up > FX up
        (CallFunction|ResetMaterialReactivity)
        ;;[1 1 1] FX
        (CallFunction|StageOne))
      (:Stage%202
        (CallFunction|UpdateMaterialReactivity 1)
        (CallFunction|StageTwo))
      (:Stage%203
        (CallFunction|UpdateMaterialReactivity 2)
        (CallFunction|StageThree)))
    (else
      ;[1 1 1] In non FE mode: Its gonna be 5 stages represents 5 eliminates
      (bind (_statname _statvalue) (Cosmetics|WatchCosmeticStat (CustomCharacter|GetAssociatedPlayerPawn) "(TagName=\"StatManager.AthenaKills\")")
        (:OnStatChanged
          (CallFunction|UpdateMaterialReactivity _statvalue)
          (Variables|Default|SetLastValue (Math|Integer|SelectInt (Math|Integer|Max(Integer) _statvalue (Variables|Default|GetLastValue)) (Variables|Default|GetEndStage) (< (Math|Integer|Max(Integer) _statvalue (Variables|Default|GetLastValue)) (+ (Variables|Default|GetEndStage) 1))))
          (Utilities|FlowControl|Sequence
            (:then_0
              (if (Variables|Default|GetIsLit)
                _
                (elif (== (Variables|Default|GetLastValue) 1)
                  (CallFunction|StageTwo)
                  (elif (== (Variables|Default|GetLastValue) 2)
                    _)))
              (CallFunction|StageThree))))))))

(event Custom|StageOne
  (bind _returnvalue (CallFunction|GetVFXComponent "(TagName=\"CosmeticPart.VFX.Idle\")"))
  ;[1 1 1] FX Stage
  ;; 0 Elims
  (Variables|Default|SetIsLit false)
  (Effects|Components|ParticleSystem|SetBoolParameter _returnvalue "isStage2" false)
  (Effects|Components|ParticleSystem|SetBoolParameter _returnvalue "isStage3" false))

(event Custom|PlayReactiveAudio
  (params (Event%20Name :type Gameplay%20Tag%20Structure)))

(event Custom|StageTwo
  ;[1 1 1] FX Stage
  ;; 1 Elim
  (Variables|Default|SetIsLit true)
  (Effects|Components|ParticleSystem|SetBoolParameter (CallFunction|GetIdleVFXComponent) "bisLit" (Variables|Default|GetIsLit))
  (Effects|Components|ParticleSystem|SetBoolParameter (CallFunction|GetVFXComponent "(TagName=\"CosmeticPart.VFX.Idle\")") "isStage2" true))

(event Custom|StageThree
  ;[1 1 1] FX Stage
  ;; 2 Elims
  (Effects|Components|ParticleSystem|SetBoolParameter (CallFunction|GetVFXComponent "(TagName=\"CosmeticPart.VFX.Idle\")") "isStage3" true)
  (Variables|Default|SetisStage3 true))

(event Custom|UpdateMaterialReactivity
  (params (KillCount :type Integer))
  (Rendering|Material|SetScalarParameterValueonMaterials (CustomCharacter|GetSkeletalMeshForPartOnPlayerPawn) "KillCount" (Math|Conversions|ToFloat(Integer) KillCount))
  (Rendering|Material|SetScalarParameterValueonMaterials (CustomCharacter|GetSkeletalMeshForPartOnPlayerPawn) "Timestamp" (Utilities|Time|GetGameTimeinSeconds)))

(event Custom|ResetMaterialReactivity
  (Rendering|Material|SetScalarParameterValueonMaterials (CustomCharacter|GetSkeletalMeshForPartOnPlayerPawn) "KillCount" 0.0))

(fn StatValueToStage
  (params (StatValue :type Integer))
  (bind (_arrayindex _arrayelement) (Utilities|Array|ReverseforEachLoop (Variables|PartModifierControls|WatchCosmeticStat|GetStat%20Thresholds))
    (:LoopBody
      (if (>= (Variables|GetStatValue) _arrayelement)
        (return _arrayindex)))
    (:Completed
      (return 0))))

(fn UpdateMaterialParameters
  (switch int (Variables|Default|GetCurrentStage)
    (:0
      (CallFunction|StageOne))
    (:2
      (CallFunction|StageTwo))))

(fn ConstructionScript
  (bind _returnvalue (CustomCharacter|GetSkeletalMeshForPartOnPlayerPawn))
  (|CallParentFunction)
  (Rendering|Material|SetScalarParameterValueonMaterials _returnvalue "GameTimeFromBP" -99999.0)
  (Rendering|Material|SetScalarParameterValueonMaterials _returnvalue "StartAtStage" (Math|Conversions|ToFloat(Integer) (Variables|Default|GetStartStage)))
  (Rendering|Material|SetScalarParameterValueonMaterials _returnvalue "StopAtStage" (Math|Conversions|ToFloat(Integer) (Variables|Default|GetEndStage))))
