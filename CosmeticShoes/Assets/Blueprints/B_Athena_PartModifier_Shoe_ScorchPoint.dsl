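; Part modifier for the Scorch Point cosmetic shoes, which heat up reactively as the player earns eliminations.
; The watched kill stat drives material stage scalars, timestamped with game time for shader animation, and lights idle FX at a threshold, firing a landing burst.
; Landing FX play only while lit and visible, and weapons tagged to hide foot FX switch off idle FX.
; In the frontend, a timeline previews the stages instead.
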(event EventOnPawnLanded
  ;[0.6233 0.4542 0 0.5] OnPawnLand Override because on this shoe we only spawn the landing effect after 3 eliminations
  (if (Variables|Default|GetIsLit)
    (if (or (Variables|HiddenPartCheck|GetbIsFootHidden) (Variables|Default|GetbAreShoesHidden))
      _
      (elif (CallFunction|IsPreventLandingFXPlayingatIncorrectTimesEnabled)
        (CustomCharacter|ShouldPlayLandingFX "(GameplayTags=((TagName=\"Gameplay.Action.Player.InVehicle\"),(TagName=\"Gameplay.Action.Player.AthenaInVehicle\"),(TagName=\"Granted.Athena.EnvItem.HidingProp.GettingInProp\")))" "(GameplayTags=((TagName=\"MovementMode.Parachuting\"),(TagName=\"MovementMode.Skydiving\")))" 3e+01
          (:True
            (bind _returnvalue (CallFunction|GetVFXSpawnParameters "(TagName=\"CosmeticPart.VFX.Landing\")")
              (:Success
                (bind _returnvalue_2 (PartModifier|SpawnSystem|SpawnSystemAttachedPartModifier (Utilities|Struct|BreakCharacterPartModifierVFXSpawnParameters _returnvalue) (.AttachToComponent (Utilities|Struct|BreakCharacterPartModifierVFXSpawnParameters _returnvalue)) (.SocketName (Utilities|Struct|BreakCharacterPartModifierVFXSpawnParameters _returnvalue)) "0, 0, 0" "0, 0, 0" (.LocationType (Utilities|Struct|BreakCharacterPartModifierVFXSpawnParameters _returnvalue)) true (.bAutoActivate (Utilities|Struct|BreakCharacterPartModifierVFXSpawnParameters _returnvalue)))
                  (:Success
                    (Niagara|SetNiagaraVariable(Vector3) _returnvalue_2 "ShoeScale_R" (Variables|Default|GetShoeScaleR))
                    (Niagara|SetNiagaraVariable(Vector3) _returnvalue_2 "ShoeScale_L" (Variables|Default|GetShoeScaleL))))))))
        (else
          _)))))

(event EventOnWeaponChanged
  (params (NewWeapon :type Fort%20Weapon%20Object%20Reference) (OldWeapon :type Fort%20Weapon%20Object%20Reference))
  (bind _returnvalue (CallFunction|GetVFXComponent "(TagName=\"CosmeticPart.VFX.Idle\")"))
  (Utilities|IsValid NewWeapon
    (:Is%20Valid
      (bind _output_get (Variables|Default|SetbAreShoesHidden (or (Variables|HiddenPartCheck|GetbIsFootHidden) (Tags|WeaponDataHasMatchingGameplayTag NewWeapon "(TagName=\"Animation.CharacterParts.HideFX.Feet\")"))))
      (if _output_get
        (Effects|Components|ParticleSystem|SetBoolParameter _returnvalue "bisLit" false)
        (elif (>= (Variables|Default|GetLastValue) 2)
          (Effects|Components|ParticleSystem|SetBoolParameter _returnvalue "bisLit" true))))))

(event EventOnCosmeticStateActivated
  (params (Stat%20Name :type Name :default "None") (Stat%20Value :type Integer) (STAGE%20Num%20Being%20Activated :type Integer))
  (|CallParentFunction Stat%20Name Stat%20Value STAGE%20Num%20Being%20Activated)
  (Variables|Default|SetCurrentStage STAGE%20Num%20Being%20Activated)
  (CallFunction|UpdateMaterialParameters))

(event EventPreDestruction
  (|CallParentFunction)
  (Utilities|Time|ClearandInvalidateTimerbyHandle (Variables|Default|GetRefreshTimeHandler)))

(event EventOnPawnAndPartChanged
  (|CallParentFunction)
  (if (not (CustomCosmeticModifier|IsinFrontEnd))
    (Utilities|FlowControl|Sequence
      (:then_0
        (bind _returnvalue (Class|FortCosmeticFunctionLibrary|GetCosmeticStatValue (CustomCharacter|GetAssociatedPlayerPawn) "(TagName=\"StatManager.AthenaKills\")"))
        (bind _stage (CallFunction|StatValuetoStage _returnvalue))
        (Variables|Default|SetCurrentStage _stage)
        (CallFunction|UpdateMaterialParameters)))))

(event EventOnPostResIn
  (bind _returnvalue (CustomCharacter|GetSkeletalMeshForPartOnPlayerPawn))
  (bind _returnvalue_2 (CustomCharacter|GetSkeletalMeshForPartOnPlayerPawn))
  (bind _endstage (Variables|Default|GetEndStage))
  (bind _returnvalue_3 (+ _endstage 1))
  (bind _lastvalue (Variables|Default|GetLastValue))
  (bind _animdurationperstage (Variables|Default|GetAnimDurationPerStage))
  (bind _lastvalue_2 (Variables|Default|GetLastValue))
  (|CallParentFunction)
  (if (CallFunction|ShouldDisplayFrontendReactivity)
    ;[1 1 1] In FE mode: Its gonna be 3 stage represents Normal > Emissive light up > FX up
    (Timeline|Timeline_0 PlayFromStart
      (:Stage%201
        ;;[1 1 1] FX
        (CallFunction|StageOne)
        (Rendering|Material|SetScalarParameterValueonMaterials _returnvalue_2 "Animation_Duration" (Utilities|Array|Get(aref) _animdurationperstage))
        (Rendering|Material|SetScalarParameterValueonMaterials _returnvalue_2 "CurrentStage" 0.0))
      (:Stage%202
        (Rendering|Material|SetScalarParameterValueonMaterials _returnvalue_2 "Animation_Duration" (Utilities|Array|Get(aref) _animdurationperstage 1))
        (Rendering|Material|SetScalarParameterValueonMaterials _returnvalue_2 "CurrentStage" 1.0))
      (:Stage%203
        (CallFunction|StageTwo)
        (Rendering|Material|SetScalarParameterValueonMaterials _returnvalue_2 "Animation_Duration" (Utilities|Array|Get(aref) _animdurationperstage 2))
        (Rendering|Material|SetScalarParameterValueonMaterials _returnvalue_2 "CurrentStage" 2.0))
      (:Stage%204
        (CallFunction|StageOne)
        (Rendering|Material|SetScalarParameterValueonMaterials _returnvalue_2 "Animation_Duration" (Utilities|Array|Get(aref) _animdurationperstage 4))
        (Rendering|Material|SetScalarParameterValueonMaterials _returnvalue_2 "CurrentStage" 5.0)))
    (CallFunction|OnActualStatChange)
    (else
      ;[1 1 1] In non FE mode: Its gonna be 5 stages represents 5 eliminates
      (bind (_statname _statvalue) (Cosmetics|WatchCosmeticStat (CustomCharacter|GetAssociatedPlayerPawn) "(TagName=\"StatManager.AthenaKills\")")
        (:OnStatChanged
          (Variables|Default|SetLastValue (Math|Integer|SelectInt (Math|Integer|Max(Integer) _statvalue _lastvalue_2) _endstage (< (Math|Integer|Max(Integer) _statvalue _lastvalue_2) _returnvalue_3)))
          (Utilities|FlowControl|Sequence
            (:then_0
              (CallFunction|InitializeVariabels)
              (Rendering|Material|SetScalarParameterValueonMaterials _returnvalue "Animation_Duration" (Utilities|Array|Get(aref) (Variables|Default|GetAnimDurationPerStage) _lastvalue))
              (Rendering|Material|SetScalarParameterValueonMaterials _returnvalue "CurrentStage" (Math|Conversions|ToFloat(Integer) _lastvalue))
              (if (Variables|Default|GetIsLit)
                _
                (elif (== (Variables|Default|GetLastValue) 2)
                  (CallFunction|StageTwo))))
            (:then_1
              (if (and (> (Math|Integer|Max(Integer) _statvalue _lastvalue_2) (Variables|Default|GetStartStage)) (< (Math|Integer|Max(Integer) _statvalue _lastvalue_2) _returnvalue_3))
                (CallFunction|OnActualStatChange)))))))))

(event Custom|OnActualStatChange
  ;[1 1 1] In non FE mode: Its gonna be 5 stages represents 5 eliminates
  ;;[1 1 1] This should only fire off when the stat changes - use this event!
  (Rendering|Material|SetScalarParameterValueonMaterials (CustomCharacter|GetSkeletalMeshForPartOnPlayerPawn) "GameTimeFromBP" (Utilities|Time|GetGameTimeinSeconds)))

(event Custom|StageOne
  ;[1 1 1] FX Stage
  (Variables|Default|SetIsLit false)
  (Effects|Components|ParticleSystem|SetBoolParameter (CallFunction|GetIdleVFXComponent) "bisLit" (Variables|Default|GetIsLit)))

(event Custom|PlayReactiveAudio
  (params (Event%20Name :type Gameplay%20Tag%20Structure)))

(event Custom|InitializeVariabels
  (bind _returnvalue (CustomCharacter|GetSkeletalMeshForPartOnPlayerPawn))
  ;[1 1 1] Initialize Variables
  (Rendering|Material|SetScalarParameterValueonMaterials _returnvalue "StartAtStage" (Math|Conversions|ToFloat(Integer) (Variables|Default|GetStartStage)))
  (Rendering|Material|SetScalarParameterValueonMaterials _returnvalue "StopAtStage" (Math|Conversions|ToFloat(Integer) (Variables|Default|GetEndStage))))

(event Custom|StageTwo
  ;[1 1 1] FX Stage
  (Variables|Default|SetIsLit true)
  (Effects|Components|ParticleSystem|SetBoolParameter (CallFunction|GetIdleVFXComponent) "bisLit" (Variables|Default|GetIsLit))
  (bind _returnvalue (CallFunction|GetVFXSpawnParameters "(TagName=\"CosmeticPart.VFX.Landing\")")
    (:Success
      (bind _returnvalue_2 (PartModifier|SpawnSystem|SpawnSystemAttachedPartModifier (Utilities|Struct|BreakCharacterPartModifierVFXSpawnParameters _returnvalue) (.AttachToComponent (Utilities|Struct|BreakCharacterPartModifierVFXSpawnParameters _returnvalue)) (.SocketName (Utilities|Struct|BreakCharacterPartModifierVFXSpawnParameters _returnvalue)) "0, 0, 0" "0, 0, 0" (.LocationType (Utilities|Struct|BreakCharacterPartModifierVFXSpawnParameters _returnvalue)) true (.bAutoActivate (Utilities|Struct|BreakCharacterPartModifierVFXSpawnParameters _returnvalue)))
        (:Success
          (Niagara|SetNiagaraVariable(Vector3) _returnvalue_2 "ShoeScale_R" (Variables|Default|GetShoeScaleR))
          (Niagara|SetNiagaraVariable(Vector3) _returnvalue_2 "ShoeScale_L" (Variables|Default|GetShoeScaleL)))))))

(fn StatValueToStage
  (params (StatValue :type Integer))
  (bind (_arrayindex _arrayelement) (Utilities|Array|ReverseforEachLoop (Variables|PartModifierControls|WatchCosmeticStat|GetStat%20Thresholds))
    (:LoopBody
      (if (>= (Variables|GetStatValue) _arrayelement)
        (return _arrayindex)))
    (:Completed
      (return 0))))

(fn UpdateMaterialParameters
  (switch int (Variables|Default|GetCurrentStage)
    (:0
      (CallFunction|StageOne))
    (:2
      (CallFunction|StageTwo))))

(fn ConstructionScript
  (bind _returnvalue (CustomCharacter|GetSkeletalMeshForPartOnPlayerPawn))
  (|CallParentFunction)
  (Rendering|Material|SetScalarParameterValueonMaterials _returnvalue "GameTimeFromBP" -99999.0)
  (Rendering|Material|SetScalarParameterValueonMaterials _returnvalue "StartAtStage" (Math|Conversions|ToFloat(Integer) (Variables|Default|GetStartStage)))
  (Rendering|Material|SetScalarParameterValueonMaterials _returnvalue "StopAtStage" (Math|Conversions|ToFloat(Integer) (Variables|Default|GetEndStage))))
