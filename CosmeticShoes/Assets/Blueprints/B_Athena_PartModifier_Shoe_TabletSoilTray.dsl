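; Part modifier for the TabletSoilTray cosmetic shoes, driving reactive materials and effects from pawn state and variant.
; Part changes cache and debug-print the material variant; Tick sets "IsRun" during sliding or tactical sprinting.
; Landing runs a timeline on "landingBlend" and spawns landing VFX for visible variant-0 shoes.
; In the front end, a delayed timeline drives that blend and the idle VFX trail; otherwise it stores the player pawn.
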(event EventOnPawnAndPartChanged
  (|CallParentFunction)
  (bind _asfort_player_pawn_athena (Utilities|Casting|CastToFortPlayerPawnAthena (CustomCharacter|GetAssociatedPlayerPawn)))
  (bind _index (CosmeticLoadout|GetVariantChannelActiveIndexforSlotActor "(TagName=\"LoadoutSlot.ExtraPart\")" _asfort_player_pawn_athena "(TagName=\"Cosmetics.Variant.Channel.Material\")" (Utilities|Array|MakeArray "(TagName=\"Cosmetics.Variant.Property.Mat1\")" "(TagName=\"Cosmetics.Variant.Property.Mat2\")")))
  (bind _output_get (Variables|Default|SetCurrent%20variant _index))
  (Development|PrintString (Utilities|String|ToString(Integer) _output_get) true true "(R=1.000000,G=0.942534,B=0.000000,A=1.000000)" 2e+01))

(event EventOnPostResIn
  ;[0 0 0 0.5] Get Shoe Mutable Scale
  (|CallParentFunction)
  (if (CustomCosmeticModifier|IsinFrontEnd)
    (Rendering|Material|SetScalarParameterValueonMaterials (CustomCharacter|GetSkeletalMeshForPartOnPlayerPawn) "FrontEnd_SpeedLine_Radius" 0.85)
    (Effects|Components|ParticleSystem|SetFloatParameter (CallFunction|GetVFXComponent "(TagName=\"CosmeticPart.VFX.Idle\")") "TrailOn" 0.0)
    (Utilities|FlowControl|Delay 3.0)
    (bind (_direction _landinganimation) (Timeline|Timeline_0 PlayFromStart)
      (:Update
        _)
      (:Finished
        _))
    (Rendering|Material|SetScalarParameterValueonMaterials (CustomCharacter|GetSkeletalMeshForPartOnPlayerPawn) "landingBlend" _landinganimation)
    (Effects|Components|ParticleSystem|SetFloatParameter (CallFunction|GetVFXComponent "(TagName=\"CosmeticPart.VFX.Idle\")") "TrailOn" _landinganimation)
    (else
      (bind _asfort_player_pawn_athena (Utilities|Casting|CastToFortPlayerPawnAthena (CustomCharacter|GetAssociatedPlayerPawn)))
      (Variables|Default|SetPlayerRef _asfort_player_pawn_athena))))

(event EventOnPawnLanded
  (bind _returnvalue (CustomCharacter|GetSkeletalMeshForPartOnPlayerPawn))
  ;[0 0 0 0.5] On land
  (|CallParentFunction)
  (Utilities|FlowControl|Sequence
    (:then_1
      (Utilities|FlowControl|Delay 0.06)
      (bind (_direction _landinganimation) (Timeline|Timeline PlayFromStart)
        (:Update
          (Rendering|Material|SetScalarParameterValueonMaterials _returnvalue "landingBlend" _landinganimation))
        (:Finished
          (Rendering|Material|SetScalarParameterValueonMaterials _returnvalue "landingBlend" 0.0))))
    (:then_2
      (if (and (== (Variables|Default|GetCurrent%20variant) 0) (!= (Variables|Default|GetbAreShoesHidden) true))
        (if (CallFunction|IsPreventLandingFXPlayingatIncorrectTimesEnabled)
          (CustomCharacter|ShouldPlayLandingFX "(GameplayTags=((TagName=\"Gameplay.Action.Player.InVehicle\"),(TagName=\"Gameplay.Action.Player.AthenaInVehicle\"),(TagName=\"Granted.Athena.EnvItem.HidingProp.GettingInProp\")))" "(GameplayTags=((TagName=\"MovementMode.Parachuting\"),(TagName=\"MovementMode.Skydiving\")))" 3e+01
            (:True
              _)))
        (bind _returnvalue_2 (CallFunction|GetVFXSpawnParameters "(TagName=\"CosmeticPart.VFX.Landing\")")
          (:Success
            (PartModifier|SpawnSystem|SpawnSystemAttachedPartModifier (Utilities|Struct|BreakCharacterPartModifierVFXSpawnParameters _returnvalue_2) (.AttachToComponent (Utilities|Struct|BreakCharacterPartModifierVFXSpawnParameters _returnvalue_2)) (.SocketName (Utilities|Struct|BreakCharacterPartModifierVFXSpawnParameters _returnvalue_2)) "0, 0, 0" "0, 0, 0" (.LocationType (Utilities|Struct|BreakCharacterPartModifierVFXSpawnParameters _returnvalue_2)) true (.bAutoActivate (Utilities|Struct|BreakCharacterPartModifierVFXSpawnParameters _returnvalue_2)))))))))

(event EventTick
  (params (DeltaSeconds :type Float%20%28single-precision%29))
  (bind _returnvalue (CustomCharacter|GetSkeletalMeshForPartOnPlayerPawn))
  (bind _playerref (Variables|Default|GetPlayerRef))
  (|CallParentFunction DeltaSeconds)
  (if (Utilities|IsValid _playerref :!class /Script/Engine.KismetSystemLibrary)
    (if (Reactive|Frontend|IsCosmeticObjectDisplayedinFrontendContext 0)
      (Actor|Tick|SetActorTickEnabled false)
      (else
        (Actor|Tick|SetActorTickEnabled true)
        (if (or (Variables|/Script/FortniteGame.FortPawn|GetbIsSliding _playerref) (Variables|/Script/FortniteGame.FortPlayerPawn|GetbIsTacticalSprinting _playerref))
          (Rendering|Material|SetScalarParameterValueonMaterials _returnvalue "IsRun" 1.0)
          (else
            (Rendering|Material|SetScalarParameterValueonMaterials _returnvalue "IsRun" 0.0)))))))

(event EventPreDestruction
  ;[1 1 1] Turn Default Dust FX back on
  (|CallParentFunction))

(fn ConstructionScript
  (|CallParentFunction))
