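; Cosmetic part modifier Blueprint for the VisionHorseGown shoe that adds a plastic overlay material to the character part.
; After post-res-in, it gets the part's skeletal mesh on the player pawn and calls SetOverlayMaterial with the shoe's plastic material instance.
; That event also has a front-end check with an empty branch.
; Landing, pre-destruction, pawn/part change and construction hooks only call their parent implementations.
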(event EventOnPawnLanded
  (|CallParentFunction))

(event EventPreDestruction
  (|CallParentFunction))

(event EventOnPawnAndPartChanged
  (|CallParentFunction))

(event EventOnPostResIn
  (|CallParentFunction)
  (if (CustomCosmeticModifier|IsinFrontEnd)
    _)
  (Rendering|Material|SetOverlayMaterial (CustomCharacter|GetSkeletalMeshForPartOnPlayerPawn) "/CosmeticShoes/Assets/FORT_Shoes/Shoe_VisionHorseGown/Materials/MI_VisionHorseGown_Shoe_Plastic.MI_VisionHorseGown_Shoe_Plastic" false 0 :!class /Script/Engine.MeshComponent))

(fn ConstructionScript
  (|CallParentFunction))
