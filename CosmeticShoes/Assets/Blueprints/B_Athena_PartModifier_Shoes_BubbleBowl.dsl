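; Part modifier for the Bubble Bowl cosmetic shoes that fits their idle VFX fish to the character's foot proportions after res-in.
; It measures left ball and foot bone offsets on the part's skeletal mesh, dividing ball length and foot height by the X shoe scale times a constant.
; Clamped range maps turn these, plus the unscaled ball height above root, into front and rear fish offset lerp parameters on the particle component.
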(event EventOnPostResIn
  (bind _returnvalue (CustomCharacter|GetSkeletalMeshForPartOnPlayerPawn))
  (bind _returnvalue_2 (Transformation|GetSocketTransform _returnvalue "root" "RTS_World"))
  (bind _shoescalel (Variables|Default|GetShoeScaleL))
  (bind _x (.x _shoescalel))
  (bind _location (.location _returnvalue_2))
  (bind _returnvalue_3 (CallFunction|GetVFXComponent "(TagName=\"CosmeticPart.VFX.Idle\")"))
  (bind _returnvalue_4 (* _x 30.63))
  (|CallParentFunction)
  (Effects|Components|ParticleSystem|SetFloatParameter _returnvalue_3 "FrontFishOffsetLerpX" (Math|Float|MapRangeClamped (/ (Math|Vector|VectorLength (Math|Vector|MakeVector 0.0 (.y (.location (Transformation|GetSocketTransform _returnvalue "ball_l" "RTS_ParentBoneSpace"))) (.z (.location (Transformation|GetSocketTransform _returnvalue "ball_l" "RTS_ParentBoneSpace"))))) _returnvalue_4) 0.3714 0.5 0.0 1.0))
  (Effects|Components|ParticleSystem|SetFloatParameter _returnvalue_3 "FrontFishOffsetLerpY" (Math|Float|MapRangeClamped (Math|Float|Absolute(Float) (.z (- (.location (Transformation|GetSocketTransform _returnvalue "ball_l" "RTS_World")) _location))) 5.0 3.0 0.0 1.0))
  (Effects|Components|ParticleSystem|SetFloatParameter _returnvalue_3 "RearFishOffsetLerpZ" (Math|Float|MapRangeClamped (/ (.z (- (.location (Transformation|GetSocketTransform _returnvalue "foot_l" "RTS_World")) _location)) _returnvalue_4) 0.285 0.38 0.0 1.0)))

(fn ConstructionScript
  (|CallParentFunction))
