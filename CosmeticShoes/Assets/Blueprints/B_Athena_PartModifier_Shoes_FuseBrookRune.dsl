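; Cosmetic shoe part modifier for FuseBrookRune that drives time-based material parameters from the player's movement.
; After initialization it flags the materials as front-end or in-game and, in-game, caches the player pawn and its slide-or-tactical-sprint state.
; While either move is active, Tick stamps RunningTimeStart with game time, plus StateChangeTime if the state just changed or the pawn is jumping.
; Tick disables itself for front-end display.
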(event EventOnPostResIn
  ;[0 0 0 0.5] Get Shoe Mutable Scale
  (|CallParentFunction)
  (if (CustomCosmeticModifier|IsinFrontEnd)
    (Rendering|Material|SetScalarParameterValueonMaterials (CustomCharacter|GetSkeletalMeshForPartOnPlayerPawn) "IsInFrontEnd" 1.0)
    (else
      (Rendering|Material|SetScalarParameterValueonMaterials (CustomCharacter|GetSkeletalMeshForPartOnPlayerPawn) "IsInFrontEnd" 0.0)
      (bind _asfort_player_pawn_athena (Utilities|Casting|CastToFortPlayerPawnAthena (CustomCharacter|GetAssociatedPlayerPawn)))
      (Variables|Default|SetPlayer%20Ref _asfort_player_pawn_athena)
      (Variables|Default|SetTickStats (or (Variables|/Script/FortniteGame.FortPawn|GetbIsSliding (Variables|Default|GetPlayer%20Ref)) (Variables|/Script/FortniteGame.FortPlayerPawn|GetbIsTacticalSprinting (Variables|Default|GetPlayer%20Ref)))))))

(event EventTick
  (params (DeltaSeconds :type Float%20%28single-precision%29))
  (bind _returnvalue (CustomCharacter|GetSkeletalMeshForPartOnPlayerPawn))
  (bind _player_ref (Variables|Default|GetPlayer%20Ref))
  (bind _bistacticalsprinting (Variables|/Script/FortniteGame.FortPlayerPawn|GetbIsTacticalSprinting _player_ref))
  (bind _bissliding (Variables|/Script/FortniteGame.FortPawn|GetbIsSliding _player_ref))
  (bind _returnvalue_2 (or _bissliding _bistacticalsprinting))
  (|CallParentFunction DeltaSeconds)
  (if (Utilities|IsValid _player_ref :!class /Script/Engine.KismetSystemLibrary)
    (if (Reactive|Frontend|IsCosmeticObjectDisplayedinFrontendContext 0)
      (Actor|Tick|SetActorTickEnabled false)
      (else
        (Actor|Tick|SetActorTickEnabled true)
        (if _returnvalue_2
          (Rendering|Material|SetScalarParameterValueonMaterials _returnvalue "RunningTimeStart" (Utilities|Time|GetGameTimeinSeconds))
          (if (not (and (== (Variables|Default|GetTickStats) _returnvalue_2) (!= (Class|FortPawn|IsJumping _player_ref) true)))
            (Rendering|Material|SetScalarParameterValueonMaterials _returnvalue "StateChangeTime" (Utilities|Time|GetGameTimeinSeconds)))
          (else
            (Utilities|FlowControl|Sequence
              (:then_0
                _))))
        (Variables|Default|SetTickStats _returnvalue_2)))))

(fn ConstructionScript
  (|CallParentFunction))
