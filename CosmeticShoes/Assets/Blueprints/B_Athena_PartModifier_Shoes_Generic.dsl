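; Generic cosmetic part modifier for shoes that drives their idle and landing VFX and turns off the pawn's movement dust until destruction.
; Footsteps briefly pulse idle-effect parameters, while valid landings spawn tagged VFX or a fallback effect unless shoes are hidden.
; Weapon changes toggle the idle effect, and async controller tag-count listeners flag active abilities to it outside the frontend.
; Morph adjustment data supplies per-foot visibility and corrected shoe scales for the effects.
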(event Custom|OnFootstep
  (params (FootstepIndex :type Integer))
  (bind _returnvalue (CallFunction|GetVFXComponent "(TagName=\"CosmeticPart.VFX.Idle\")"))
  ;[0.3523 0.7274 0.525 0.5] OnFootstep - WIP
  (Effects|Components|ParticleSystem|SetIntParameter _returnvalue "FootstepIndex" FootstepIndex)
  (Effects|Components|ParticleSystem|SetFloatParameter _returnvalue "Footstep" 1.0)
  (Utilities|FlowControl|Delay 0.04)
  (Effects|Components|ParticleSystem|SetFloatParameter _returnvalue "Footstep" 0.0))

(event EventOnPostResIn
  ;[0.1984 0.2865 0.2657 0.5] Disabling Dust FX
  (|CallParentFunction)
  (Variables|/Script/FortniteGame.FortPlayerPawn|SetbFXPlayDustOnMovement false (CustomCharacter|GetAssociatedPlayerPawn))
  (Effects|Components|ParticleSystem|SetBoolParameter (CallFunction|GetVFXComponent "(TagName=\"CosmeticPart.VFX.Idle\")") "bisPrimaryPreviewItem" (and (Reactive|Frontend|IsCosmeticObjectonPawn 0) (CallFunction|ShouldDisplayFrontendReactivity))))

(event EventOnWeaponChanged
  (params (NewWeapon :type Fort%20Weapon%20Object%20Reference) (OldWeapon :type Fort%20Weapon%20Object%20Reference))
  (bind _returnvalue (CallFunction|GetVFXComponent "(TagName=\"CosmeticPart.VFX.Idle\")"))
  ;[0.4878 0.2406 0.239 0.5] HideShoeFX
  (Utilities|IsValid NewWeapon
    (:Is%20Valid
      (bind _output_get (Variables|Default|SetbAreShoesHidden (or (Variables|HiddenPartCheck|GetbIsFootHidden) (Tags|WeaponDataHasMatchingGameplayTag NewWeapon "(TagName=\"Animation.CharacterParts.HideFX.Feet\")"))))
      (if _output_get
        (Components|Activation|Deactivate _returnvalue)
        ;[1 1 1] We'll probably remove those booleans in some point
        (Effects|Components|ParticleSystem|SetBoolParameter _returnvalue "HideFootFX" true)
        (else
          (Components|Activation|Activate _returnvalue false)
          (Effects|Components|ParticleSystem|SetBoolParameter _returnvalue "HideFootFX" false))))))

(event EventPreDestruction
  ;[0.1984 0.2865 0.2657 0.5] Disabling Dust FX
  (|CallParentFunction)
  (Variables|/Script/FortniteGame.FortPlayerPawn|SetbFXPlayDustOnMovement true (CustomCharacter|GetAssociatedPlayerPawn)))

(event EventOnPawnLanded
  (bind _landingfx (Variables|Default|GetLandingFX))
  ;[0.5668 0.6094 0.2127 0.5] Spawn FX On Land
  (if (or (Variables|HiddenPartCheck|GetbIsFootHidden) (Variables|Default|GetbAreShoesHidden))
    _
    (elif (CallFunction|IsPreventLandingFXPlayingatIncorrectTimesEnabled)
      (CustomCharacter|ShouldPlayLandingFX "(GameplayTags=((TagName=\"Gameplay.Action.Player.InVehicle\"),(TagName=\"Gameplay.Action.Player.AthenaInVehicle\"),(TagName=\"Granted.Athena.EnvItem.HidingProp.GettingInProp\")))" "(GameplayTags=((TagName=\"MovementMode.Parachuting\"),(TagName=\"MovementMode.Skydiving\")))" 3e+01
        (:True
          (Effects|Components|ParticleSystem|SetFloatParameter (CallFunction|GetVFXComponent "(TagName=\"CosmeticPart.VFX.Idle\")") "TimeStamp" (Utilities|Time|GetGameTimeinSeconds))
          (bind _returnvalue (CallFunction|GetVFXSpawnParameters "(TagName=\"CosmeticPart.VFX.Landing\")")
            (:Success
              (bind _returnvalue_2 (PartModifier|SpawnSystem|SpawnSystemAttachedPartModifier (Utilities|Struct|BreakCharacterPartModifierVFXSpawnParameters _returnvalue) (.AttachToComponent (Utilities|Struct|BreakCharacterPartModifierVFXSpawnParameters _returnvalue)) (.SocketName (Utilities|Struct|BreakCharacterPartModifierVFXSpawnParameters _returnvalue)) "0, 0, 0" "0, 0, 0" (.LocationType (Utilities|Struct|BreakCharacterPartModifierVFXSpawnParameters _returnvalue)) true (.bAutoActivate (Utilities|Struct|BreakCharacterPartModifierVFXSpawnParameters _returnvalue)))
                (:Success
                  (CallFunction|SetShoeScalesonVFX _returnvalue_2))))
            (:Failure
              (Utilities|IsValid _landingfx
                (:Is%20Valid
                  (bind _returnvalue_3 (PartModifier|SpawnSystem|SpawnSystemAttachedPartModifier _landingfx (CustomCharacter|GetSkeletalMeshForPartOnPlayerPawn) "root")
                    (:Success
                      (CallFunction|SetShoeScalesonVFX _returnvalue_3)))))))))
      (else
        _))))

(event EventOnPawnAndPartChanged
  (bind (_outmorphadjustmentdata _returnvalue) (CallFunction|GetActiveMorphAdjustmentDataUsedbyPart "ExtraPart"))
  (bind _returnvalue_2 (CustomCharacter|GetAssociatedPlayerPawn))
  (bind _returnvalue_3 (Fortnite|GetFortPlayerControllerfromActor _returnvalue_2))
  (bind _activeabilitytags (Variables|Default|GetActiveAbilityTags))
  (|CallParentFunction)
  (Utilities|FlowControl|Sequence
    (:then_0
      ;[0.1669 0.09592 0.2101 0.5] Disable FX Based on Gameplay Tag
      (if (not (CustomCosmeticModifier|IsinFrontEnd))
        (Utilities|IsValid _returnvalue_2
          (:Is%20Valid
            (Utilities|IsValid _returnvalue_3
              (:Is%20Valid
                (for _array_element (Variables|Default|GetGameplayTagsThatDisableFX)
                  (bind (_newtagcount _tagtriggered) (Gameplay|GameplayTags|Async|WaitGameplayTagCountChangedOnController 0 _returnvalue_3 _array_element (Variables|Default|GetTagCountParams))
                    (:OnTagCountChanged
                      (if (> _newtagcount 0)
                        (Utilities|Array|Add _activeabilitytags _tagtriggered)
                        (else
                          (Variables|Default|SetActiveAbilityTags (CallFunction|RemoveSingle _activeabilitytags _tagtriggered))))
                      (bind _output_get (Variables|Default|SetGameplay%20Ability%20Active (> (Utilities|Array|Length (Variables|Default|GetActiveAbilityTags)) 0)))
                      (Effects|Components|ParticleSystem|SetBoolParameter (CallFunction|GetVFXComponent "(TagName=\"CosmeticPart.VFX.Idle\")") "HasGameplayAbilityActive" _output_get))))))))))
    (:then_1
      ;[1 1 1] Get Shoe Mutable Scale
      (if _returnvalue
        (for _array_element_2 _outmorphadjustmentdata
          (if (Utilities|IsValidSoftObjectReference _array_element_2)
            (CallFunction|SetShoeHiddenVariables (Utilities|ResolveSoftReference _array_element_2))
            (for _array_element_3 (Variables|/Script/FortniteGame.CosmeticMorphAdjustmentData|GetTransformParameters (Utilities|ResolveSoftReference _array_element_2))
              (switch string (Utilities|Struct|BreakCosmeticCustomizableObjectTransformParameter _array_element_3)
                (:LeftShoeTransform
                  (bind _corrected_left_shoe_scale (Utilities|SetCorrectedLeftShoeScale (.scale (.ParameterValue (Utilities|Struct|BreakCosmeticCustomizableObjectTransformParameter _array_element_3)))))
                  (CallFunction|SetShoeIdleVFXScale "ShoeScale_L" _corrected_left_shoe_scale))
                (:RightShoeTransform
                  (bind _corrected_right_shoe_scale (Utilities|SetCorrectedRightShoeScale (.scale (.ParameterValue (Utilities|Struct|BreakCosmeticCustomizableObjectTransformParameter _array_element_3)))))
                  (CallFunction|SetShoeIdleVFXScale "ShoeScale_R" _corrected_right_shoe_scale))))))))))

(fn RemoveSingle
  (params (Array :type Array%20of%20Gameplay%20Tag%20Structures) (ValueToRemove :type Gameplay%20Tag%20Structure))
  (bind (_array_element _array_index) (Utilities|Array|ForEachLoopwithBreak Array)
    (:LoopBody
      (if (== _array_element (Variables|GetValueToRemove))
        (LocalVariables|SetFoundEntry _array_index)))
    (:Completed
      _))
  (if (> (LocalVariables|GetFoundEntry) -1)
    (Utilities|Array|RemoveIndex (Variables|GetArray) (LocalVariables|GetFoundEntry))
    (else
      _))
  (return (Variables|GetArray)))

(fn SetShoeHiddenVariables
  (params (CosmeticMorphAdjustmentData :type Cosmetic%20Morph%20Adjustment%20Data%20Object%20Reference))
  (switch Utilities|FlowControl|Switch|SwitchonESidedVisibility (Variables|/Script/FortniteGame.CosmeticMorphAdjustmentData|GetVisibility CosmeticMorphAdjustmentData)
    (:LeftOnly
      (Variables|Default|SetbIsLeftShoeHidden false)
      (Variables|Default|SetbIsRightShoeHidden true))
    (:RightOnly
      (Variables|Default|SetbIsLeftShoeHidden true)
      (Variables|Default|SetbIsRightShoeHidden false))
    (:Both
      (Variables|Default|SetbIsLeftShoeHidden false)
      (Variables|Default|SetbIsRightShoeHidden false))))

(fn SetShoeIdleVFXScale
  (params (ShoeParameterName :type Name) (ShoeScale :type Vector))
  (Effects|Components|ParticleSystem|SetVectorParameter (CallFunction|GetVFXComponent "(TagName=\"CosmeticPart.VFX.Idle\")") ShoeParameterName ShoeScale)
  (return))

(fn SetShoeScalesOnVFX
  (params (NiagaraParticleSystemComponent :type Niagara%20Particle%20System%20Component%20Object%20Reference))
  (Niagara|SetNiagaraVariable(Vector3) NiagaraParticleSystemComponent "ShoeScale_R" (Variables|Default|GetShoeScaleR))
  (Niagara|SetNiagaraVariable(Vector3) NiagaraParticleSystemComponent "ShoeScale_L" (Variables|Default|GetShoeScaleL)))

(fn ConstructionScript
  (|CallParentFunction))
