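; Construction script for the music-reactive variant of the generic Athena shoes part modifier, a subclass of B_Athena_PartModifier_Shoes_Generic.
; The script only calls its parent, as shoe and audio reactivity logic is kept in the base classes.
; Its comment states that this class sets the Has Audio Reactivity parameter true, which gates shoe-specific reactivity in the shoes base class.
; The comment forbids adding support here to avoid divergence from other shoe mutators.
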(fn ConstructionScript
  ;[0.6 0 0 0.5] DO NOT ADD BLUEPRINT SUPPORT HERE!!!
  ; General Shoes support goes in B_Athena_PartModifier_Shoes_Generic base class.
  ; Shoe-specific Audio Reactivity should ALSO go in B_Athena_PartModifier_Shoes_Generic,
  ; where it can be gated using Has Audio Reactivity parameter, which is set true here but
  ; false in the base class.
  ; General Audio Reactivity support goes in the base-most class B_Athena_PartModifier_Generic.
  ; If those classes have changed, search for appropriate classes, but don't add support here where
  ; it's likely to cause divergence with other shoes mutators.
  ; If you have any questions, please ask the Cosmetics Platform team.
  (|CallParentFunction))
