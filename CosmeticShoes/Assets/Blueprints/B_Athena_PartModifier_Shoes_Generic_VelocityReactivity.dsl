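; Generic cosmetic shoe part modifier that turns the player pawn's velocity into reactive events, left empty here as hooks.
; Each tick it reads velocity, optionally flattened to XY, and by flag fires linear (normalized), accumulated or moving-toggle events.
; Accumulated velocity weights walk and sprint differently, clamps and decays via FInterpTo; tactical sprint fires only on state change.
; Swimming, skydiving and sliding set tick from per-state toggles, re-enabling it when they end.
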(event EventOnBeginSwimming
  ;[1 1 1] Comment
  (|CallParentFunction)
  (Actor|Tick|SetActorTickEnabled (Variables|TickToggles|GetTickOnSwim)))

(event EventOnEndSwimming
  ;[1 1 1] Comment
  (|CallParentFunction)
  (Actor|Tick|SetActorTickEnabled true))

(event EventOnBeginSkydiving
  ;[1 1 1] Comment
  (|CallParentFunction)
  (Actor|Tick|SetActorTickEnabled (Variables|TickToggles|GetTickOnSkydive)))

(event Custom|CustomEvent
  (|CallParentFunction)
  (Development|PrintString (Utilities|String|Append "Enabling Tick for: " (Utilities|GetDisplayName self)) false true "(R=1.000000,G=0.455398,B=0.000000,A=1.000000)")
  (Actor|Tick|SetActorTickEnabled true))

(event EventOnEndSkydiving
  ;[1 1 1] Comment
  (|CallParentFunction)
  (Actor|Tick|SetActorTickEnabled true))

(event EventOnBeginSliding
  ;[1 1 1] Comment
  (|CallParentFunction)
  (Actor|Tick|SetActorTickEnabled (Variables|TickToggles|GetTickOnSliding)))

(event EventOnEndSliding
  ;[1 1 1] Comment
  (|CallParentFunction)
  (Actor|Tick|SetActorTickEnabled true))

(event EventOnPostResIn
  (|CallParentFunction))

(event EventTick
  (params (DeltaSeconds :type Float%20%28single-precision%29))
  (|CallParentFunction DeltaSeconds)
  (Utilities|FlowControl|Sequence
    (:then_0
      ;[1 1 1] GetCharacterVelocity
      (Variables|Velocity|SetPawnVelocity (Math|Vector|SelectVector (* (Transformation|GetVelocity (CustomCharacter|GetAssociatedPlayerPawn)) "1.000000,1.000000,0.000000") (Transformation|GetVelocity (CustomCharacter|GetAssociatedPlayerPawn)) (Variables|Default-Velocity|GetAllowVerticalVelocity))))
    (:then_1
      ;[0.3166 1 0.4028] Example 1 - Simple Velocity
      (if (Variables|Default-Velocity|GetUseLinearVelocity)
        (bind (_normalizedoverallvelocity _normalizedwalkvelocity _velocitylength) (Default|GetNormalizedVelocity (Variables|Velocity|GetPawnVelocity)))
        (CallFunction|OnLinearVelocity _normalizedoverallvelocity _normalizedwalkvelocity _velocitylength)))
    (:then_2
      ;[1 0.3372 0.6707] Example 2 - Accumulated Velocity
      (if (Variables|Default-Velocity|GetUseAccumulatedVelocity)
        (bind (_normalized_accumulatedvelocity _accumulated_velocity) (Default|GetNormalizedAccumulatedVelocity (Utilities|Time|GetWorldDeltaSeconds) (stale _output_get) 5e+04 0.75 2.25 (Variables|Velocity|GetDecaySpeed) (Variables|Velocity|GetPawnVelocity)))
        (bind _output_get (Variables|Velocity|SetAccumulated%20Velocity _accumulated_velocity))
        (CallFunction|OnAccumulatedVelocity _normalized_accumulatedvelocity _output_get)))
    (:then_3
      ;[0.2353 0.5182 1] Example 3 - Simple Toggle
      (if (Variables|Default-Velocity|GetUseToggleVelocity)
        (CallFunction|OnVelocityToggle (Math|Conversions|ToBoolean(Integer) (Math|Integer|Clamp(Integer) (Math|Float|Round (Math|Vector|VectorLength (Variables|Velocity|GetPawnVelocity))) 0 1)) DeltaSeconds)))
    (:then_4
      ; TacticalSprinting
      (if (Variables|Default-Velocity|GetUseTacticalSprintingEvents)
        (CallFunction|TacticalSprintingListhener)))))

(event Custom|OnTacticalSprintingToggle
  (params (Is%20TacticalSprinting :type Boolean)))

(event Custom|OnLinearVelocity
  (params (NormalizedOverallVelocity :type Float%20%28double-precision%29) (NormalizedWalkVelocity :type Float%20%28double-precision%29) (VelocityLength :type Float%20%28double-precision%29)))

(event Custom|OnAccumulatedVelocity
  (params (NormalizedAccumulatedVelocity :type Float%20%28double-precision%29) (AccumulatedVelocity :type Float%20%28double-precision%29)))

(event Custom|OnVelocityToggle
  (params (IsMoving :type Boolean) (Delta%20Seconds :type Float%20%28double-precision%29)))

(fn GetNormalizedAccumulatedVelocity
  (params (DeltaSeconds :type Float%20%28double-precision%29%20%28by%20ref%29) (LastAccumulatedVelocity :type Float%20%28double-precision%29%20%28by%20ref%29) (MaxAccumulatedVelocity :type Float%20%28double-precision%29 :default 5e+04) (WalkAccumMultiplier :type Float%20%28double-precision%29 :default 0.75) (SprintAccumMultiplier :type Float%20%28double-precision%29 :default 2.25) (DecaySpeed :type Float%20%28double-precision%29 :default 0.3) (Velocity :type Vector))
  (LocalVariables|SetVelocityVectorLength (Math|Vector|VectorLength Velocity))
  (return (* (Math|Interpolation|FInterpTo (Math|Float|Clamp(Float) (+ (Math|Float|Lerp (* (LocalVariables|GetVelocityVectorLength) (Variables|GetWalkAccumMultiplier)) (* (LocalVariables|GetVelocityVectorLength) (Variables|GetSprintAccumMultiplier)) (Math|Conversions|ToFloat(Boolean) (> (LocalVariables|GetVelocityVectorLength) (LocalVariables|GetSprintTrigger)))) (Variables|GetLastAccumulatedVelocity)) 0.0 (Variables|GetMaxAccumulatedVelocity)) 0.0 (Variables|GetDeltaSeconds) (Variables|Velocity|GetDecaySpeed)) (/ 1.0 (Variables|GetMaxAccumulatedVelocity))) (Math|Interpolation|FInterpTo (Math|Float|Clamp(Float) (+ (Math|Float|Lerp (* (LocalVariables|GetVelocityVectorLength) (Variables|GetWalkAccumMultiplier)) (* (LocalVariables|GetVelocityVectorLength) (Variables|GetSprintAccumMultiplier)) (Math|Conversions|ToFloat(Boolean) (> (LocalVariables|GetVelocityVectorLength) (LocalVariables|GetSprintTrigger)))) (Variables|GetLastAccumulatedVelocity)) 0.0 (Variables|GetMaxAccumulatedVelocity)) 0.0 (Variables|GetDeltaSeconds) (Variables|Velocity|GetDecaySpeed))))

(fn GetNormalizedVelocity
  (params (Velocity :type Vector))
  (LocalVariables|SetVelocityVectorLength (Math|Vector|VectorLength Velocity))
  (return (Math|Float|Clamp(Float) (* (/ 1.0 (LocalVariables|GetSprintTrigger)) (LocalVariables|GetVelocityVectorLength))) (Math|Float|Clamp(Float) (* (/ 1.0 (LocalVariables|GetMaxWalkSpeed)) (LocalVariables|GetVelocityVectorLength))) (LocalVariables|GetVelocityVectorLength)))

(fn TacticalSprintingListhener
  (bind _returnvalue (CustomCharacter|GetAssociatedPlayerPawn))
  (bind _bistacticalsprinting (Variables|/Script/FortniteGame.FortPlayerPawn|GetbIsTacticalSprinting _returnvalue))
  (Utilities|FlowControl|Sequence
    (:then_0
      (if (not (== _bistacticalsprinting (Variables|CharacterStates|GetIs%20Tactical%20Sprinting)))
        (CallFunction|OnTacticalSprintingToggle _bistacticalsprinting)))
    (:then_1
      (Variables|CharacterStates|SetIs%20Tactical%20Sprinting _bistacticalsprinting))))

(fn ConstructionScript
  (|CallParentFunction))
