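; Shoe cosmetic part modifier for "HintLimpFries" that drives reactive material parameters on the player pawn's part mesh.
; On post-res-in it flags the materials as on-character and frontend or in-game, playing a preview timeline into IsRun in frontend, otherwise caching the pawn.
; Tick writes game time to RunningTimeStart while sliding or tactical sprinting, leaving timing to the material.
; Tick disables itself if the cached pawn is a frontend display.
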(event EventOnPostResIn
  (bind _returnvalue (CustomCharacter|GetSkeletalMeshForPartOnPlayerPawn))
  (|CallParentFunction)
  (if (CustomCosmeticModifier|IsinFrontEnd)
    (Rendering|Material|SetScalarParameterValueonMaterials _returnvalue "OnCharacters" 1.0)
    (Rendering|Material|SetScalarParameterValueonMaterials _returnvalue "IsInFrontend" 1.0)
    (bind (_direction _runningreactivity) (Timeline|IsRunPreview PlayFromStart)
      (:Update
        (Rendering|Material|SetScalarParameterValueonMaterials _returnvalue "IsRun" _runningreactivity)))
    (else
      (bind _asfort_player_pawn_athena (Utilities|Casting|CastToFortPlayerPawnAthena (CustomCharacter|GetAssociatedPlayerPawn)))
      (Variables|Default|SetPlayer%20Ref _asfort_player_pawn_athena)
      (Rendering|Material|SetScalarParameterValueonMaterials _returnvalue "OnCharacters" 1.0)
      (Rendering|Material|SetScalarParameterValueonMaterials _returnvalue "IsInFrontend" 0.0))))

(event EventTick
  (params (DeltaSeconds :type Float%20%28single-precision%29))
  (bind _player_ref (Variables|Default|GetPlayer%20Ref))
  (|CallParentFunction DeltaSeconds)
  (if (Utilities|IsValid _player_ref :!class /Script/Engine.KismetSystemLibrary)
    (if (Reactive|Frontend|IsCosmeticObjectDisplayedinFrontendContext _player_ref)
      (Actor|Tick|SetActorTickEnabled false)
      (else
        (Actor|Tick|SetActorTickEnabled true)
        (if (or (Variables|/Script/FortniteGame.FortPawn|GetbIsSliding _player_ref) (Variables|/Script/FortniteGame.FortPlayerPawn|GetbIsTacticalSprinting _player_ref))
          (Rendering|Material|SetScalarParameterValueonMaterials (CustomCharacter|GetSkeletalMeshForPartOnPlayerPawn) "RunningTimeStart" (Utilities|Time|GetGameTimeinSeconds)))))))

(fn ConstructionScript
  (|CallParentFunction))
