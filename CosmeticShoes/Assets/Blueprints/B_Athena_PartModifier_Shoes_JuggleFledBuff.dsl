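; Cosmetic shoe part modifier for JuggleFledBuff that plays a landing Niagara effect on the frontend.
; After the parent's post-res-in logic, it checks that the cosmetic is on the pawn and frontend reactivity should display.
; If both hold, it waits briefly, then spawns the system attached to the part's skeletal mesh.
; It passes the primary-preview flag and left and right shoe scale vectors to the effect.
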(event EventOnPostResIn
  (bind _returnvalue (Reactive|Frontend|IsCosmeticObjectonPawn 0))
  (bind _returnvalue_2 (CallFunction|ShouldDisplayFrontendReactivity))
  (bind _returnvalue_3 (and _returnvalue _returnvalue_2))
  (|CallParentFunction)
  (if _returnvalue_3
    (Utilities|FlowControl|Delay 4.0)
    (bind _returnvalue_4 (Niagara|SpawnSystemAttached "/CosmeticShoes/Assets/FORT_Shoes/Shoe_JuggleFledBuff/FX/NS_Shoe_JuggleFledBuff_Landing.NS_Shoe_JuggleFledBuff_Landing" (CustomCharacter|GetSkeletalMeshForPartOnPlayerPawn)))
    (Effects|Components|ParticleSystem|SetBoolParameter _returnvalue_4 "bisPrimaryPreviewItem" _returnvalue_3)
    (Effects|Components|ParticleSystem|SetVectorParameter _returnvalue_4 "ShoeScale_R" (Variables|Default|GetShoeScaleR))
    (Effects|Components|ParticleSystem|SetVectorParameter _returnvalue_4 "ShoeScale_L" (Variables|Default|GetShoeScaleL))))

(fn ConstructionScript
  (|CallParentFunction))
