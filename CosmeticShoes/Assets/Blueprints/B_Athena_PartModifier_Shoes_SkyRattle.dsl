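; Cosmetic part modifier for the SkyRattle shoes that escalates material, audio and VFX as the player's kills rise.
; In gameplay, a watched kill stat feeds Kills, which sets the "KillCount" material scalar, calls AudioTrigger and activates idle VFX from three kills.
; In frontend displays, a looping timer cycles the count to preview stages instead.
; Landing at three or more kills spawns scaled Niagara FX unless the shoes are hidden.
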(event Custom|FrontEnd%20Cycle
  (switch int (Variables|Default|GetFrontEndInt)
    (:0
      _)
    (:1
      _)
    (:2
      _)
    (:3
      _))
  (bind _output_get (Variables|Default|SetFrontEndInt (+ (Variables|Default|GetFrontEndInt) 1)))
  (Variables|Default|SetKillCount _output_get)
  (CallFunction|Kills)
  (if (>= (Variables|Default|GetFrontEndInt) 4)
    (bind _output_get_2 (Variables|Default|SetFrontEndInt 0))
    (Variables|Default|SetKillCount _output_get_2)))

(event EventOnPostResIn
  (bind _returnvalue (CallFunction|GetVFXComponent "(TagName=\"CosmeticPart.VFX.Idle\")"))
  (|CallParentFunction)
  (Rendering|Material|SetScalarParameterValueonMaterials (CustomCharacter|GetSkeletalMeshForPartOnPlayerPawn) "KillCount" 1.0)
  ;[0.15 0.15 0.15 0.5 10] Deactivates system on start up so that it will only get activated when needed during gameplay.
  (Utilities|FlowControl|Delay 0.1)
  (Components|Activation|Deactivate _returnvalue)
  (if (Reactive|Frontend|IsCosmeticObjectDisplayedinFrontendContext 0)
    ;[0.15 0.15 0.15 0.5 10] Activates system so the switch between the on and off pawn version can happen.
    (Utilities|FlowControl|Delay 0.1)
    (Components|Activation|Activate _returnvalue false)
    (Rendering|Material|SetScalarParameterValueonMaterials (CustomCharacter|GetSkeletalMeshForPartOnPlayerPawn) "KillCount" 1.0)
    (Utilities|Time|SetTimerbyEvent (EventDispatchers|CreateEvent :!event Custom|FrontEnd%20Cycle) 6.0 true)
    (else
      (Rendering|Material|SetScalarParameterValueonMaterials (CustomCharacter|GetSkeletalMeshForPartOnPlayerPawn) "KillCount" 0.0)
      (bind (_statname _statvalue) (Cosmetics|WatchCosmeticStat (CustomCharacter|GetAssociatedPlayerPawn) "(TagName=\"StatManager.AthenaKills\")" 0 false)
        (:OnStatChanged
          (bind _output_get (Variables|Default|SetKillCount _statvalue))
          (if (> _output_get 0)
            (Variables|Default|SetIsKilled true)
            (CallFunction|Kills)))))))

(event Custom|Kills
  (bind _returnvalue (CallFunction|GetVFXComponent "(TagName=\"CosmeticPart.VFX.Idle\")"))
  (bind _killcount (Variables|Default|GetKillCount))
  (Rendering|Material|SetScalarParameterValueonMaterials (CustomCharacter|GetSkeletalMeshForPartOnPlayerPawn) "KillCount" (Math|Conversions|ToFloat(Integer) (Variables|Default|GetKillCount)))
  (Utilities|FlowControl|Sequence
    (:then_0
      (CallFunction|AudioTrigger "(TagName=\"SoundLibrary.Cosmetic.Kills\")"))
    (:then_1
      ; Activate Idle When kills = 3 or higher. Final stage activates lights and shoezaps.
      (if (> _killcount 2)
        (Components|Activation|Activate _returnvalue false)
        (Effects|Components|ParticleSystem|SetIntParameter _returnvalue "KillCountInt" _killcount)))))

(event Custom|Reset%20to%20Default
  (bind _returnvalue (CallFunction|GetVFXComponent "(TagName=\"CosmeticPart.VFX.Idle\")"))
  (Variables|Default|SetIsKilled false)
  (Rendering|Material|SetScalarParameterValueonMaterials (CustomCharacter|GetSkeletalMeshForPartOnPlayerPawn) "KillCount" 0.0)
  ; Activate Idle When kills = 3 or higher. Final stage activates lights and shoezaps.
  (Components|Activation|Deactivate _returnvalue)
  (Effects|Components|ParticleSystem|SetIntParameter _returnvalue "KillCountInt" 0))

(event EventOnPawnLanded
  ; Landing FX Override. Only spawn when kills = 3 or higher
  (|CallParentFunction)
  (if (or (Variables|HiddenPartCheck|GetbIsFootHidden) (Variables|Default|GetbAreShoesHidden))
    _
    (elif (and (> (Variables|Default|GetKillCount) 2) (Variables|Default|GetIsKilled))
      (if (CallFunction|IsPreventLandingFXPlayingatIncorrectTimesEnabled)
        (CustomCharacter|ShouldPlayLandingFX "(GameplayTags=((TagName=\"Gameplay.Action.Player.InVehicle\"),(TagName=\"Gameplay.Action.Player.AthenaInVehicle\"),(TagName=\"Granted.Athena.EnvItem.HidingProp.GettingInProp\")))" "(GameplayTags=((TagName=\"MovementMode.Parachuting\"),(TagName=\"MovementMode.Skydiving\")))" 3e+01
          (:True
            _))
        (else
          _))
      (bind _returnvalue (CallFunction|GetVFXSpawnParameters "(TagName=\"CosmeticPart.VFX.Landing\")")
        (:Success
          (bind _returnvalue_2 (PartModifier|SpawnSystem|SpawnSystemAttachedPartModifier (Utilities|Struct|BreakCharacterPartModifierVFXSpawnParameters _returnvalue) (.AttachToComponent (Utilities|Struct|BreakCharacterPartModifierVFXSpawnParameters _returnvalue)) (.SocketName (Utilities|Struct|BreakCharacterPartModifierVFXSpawnParameters _returnvalue)) "0, 0, 0" "0, 0, 0" (.LocationType (Utilities|Struct|BreakCharacterPartModifierVFXSpawnParameters _returnvalue)) true (.bAutoActivate (Utilities|Struct|BreakCharacterPartModifierVFXSpawnParameters _returnvalue)))
            (:Success
              (Niagara|SetNiagaraVariable(Vector3) _returnvalue_2 "ShoeScale_R" (Variables|Default|GetShoeScale_R))
              (Niagara|SetNiagaraVariable(Vector3) _returnvalue_2 "ShoeScale_L" (Variables|Default|GetShoeScale_L)))))))))

(fn AudioTrigger
  (params (SoundLibTag :type Sound%20Library%20Tag%20Structure))
  (return false))

(fn ConstructionScript
  (|CallParentFunction))
