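; Part modifier for the SoleVapor cosmetic shoes that adds front-end landing effects and material color adjustments after res-in.
; If the cosmetic object check and frontend reactivity check pass, a delayed Niagara landing effect attaches to the part mesh, receiving primary-preview and left/right shoe scale parameters.
; In the front end, it also creates dynamic instances of the mesh materials, giving those without PBR_Shading at 1 an inverse LUT for Inkville characters.
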(event EventOnPostResIn
  (bind _returnvalue (Reactive|Frontend|IsCosmeticObjectonPawn 0))
  (bind _returnvalue_2 (CustomCharacter|GetSkeletalMeshForPartOnPlayerPawn))
  (bind _returnvalue_3 (CallFunction|ShouldDisplayFrontendReactivity))
  (bind _returnvalue_4 (and _returnvalue _returnvalue_3))
  (|CallParentFunction)
  (Utilities|FlowControl|Sequence
    (:then_0
      (if _returnvalue_4
        (Utilities|FlowControl|Delay 4.0)
        (bind _returnvalue_5 (Niagara|SpawnSystemAttached "/CosmeticShoes/Assets/FORT_Shoes/Shoe_SoleVapor/FX/NS_Shoe_SoleVapor_Landing.NS_Shoe_SoleVapor_Landing" (CustomCharacter|GetSkeletalMeshForPartOnPlayerPawn)))
        (Effects|Components|ParticleSystem|SetBoolParameter _returnvalue_5 "bisPrimaryPreviewItem" _returnvalue_4)
        (Effects|Components|ParticleSystem|SetVectorParameter _returnvalue_5 "ShoeScale_R" (Variables|Default|GetShoeScaleR))
        (Effects|Components|ParticleSystem|SetVectorParameter _returnvalue_5 "ShoeScale_L" (Variables|Default|GetShoeScaleL))))
    (:then_1
      ;[1 1 1] Front End LUT for inkville characters
      (if (CustomCosmeticModifier|IsinFrontEnd)
        (for (_array_element _array_index) (Rendering|Material|GetMaterials _returnvalue_2)
          (bind _returnvalue_6 (Rendering|Material|CreateDynamicMaterialInstance _returnvalue_2 _array_index _array_element "None" :!class /Script/Engine.PrimitiveComponent))
          (bind _returnvalue_7 (Rendering|Material|GetScalarParameterValue _returnvalue_6 "PBR_Shading" :!class /Script/Engine.MaterialInstanceDynamic))
          (if (not (== _returnvalue_7 1.0))
            (Rendering|Material|SetTextureParameterValue _returnvalue_6 "LUT Volume Texture" "/Game/Packages/Fortress_SharedMaterials/Anime_Master_Material/VT_InverseLUT_FE.VT_InverseLUT_FE")
            (Rendering|Material|SetScalarParameterValue _returnvalue_6 "LUT Multiplier" 1.1 :!class /Script/Engine.MaterialInstanceDynamic)
            (Rendering|Material|SetScalarParameterValue _returnvalue_6 "Mobile LUT Brightness" 1.0 :!class /Script/Engine.MaterialInstanceDynamic)
            (Rendering|Material|SetScalarParameterValue _returnvalue_6 "use 33 Resolution" 0.0 :!class /Script/Engine.MaterialInstanceDynamic)
            (Rendering|Material|SetScalarParameterValue _returnvalue_6 "use Color Adjustment Instead of LUT" 0.0 :!class /Script/Engine.MaterialInstanceDynamic)))))))

(fn ConstructionScript
  (|CallParentFunction))
